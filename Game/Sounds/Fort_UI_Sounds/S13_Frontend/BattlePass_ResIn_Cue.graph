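# Sound cue for the Season 13 frontend Battle Pass "ResIn" UI effect, built from cosmetic-appear waves in the UI sound class.
# A Random node picks among six equally weighted wave variations, then a Modulator varies pitch and volume on each play.
# A concurrency override allows one instance, stopping the oldest with a short release and a retrigger delay to keep rapid repeats from piling up.
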SoundCue /Game/Sounds/Fort_UI_Sounds/S13_Frontend/BattlePass_ResIn_Cue
root=Modulator_1
nodes [
 {id=Modulator_1, children [Random_0], properties {pitchMax=1.1, volumeMax=1, volumeMin=0.5}}
 {
  id=Random_0
  children [WavePlayer_1,WavePlayer_2,WavePlayer_3,WavePlayer_4,WavePlayer_5,WavePlayer_6]
  properties {bShouldExcludeFromBranchCulling=true, weights [1,1,1,1,1,1]}
 }
 {id=WavePlayer_1, properties {soundWaveAssetPtr=/Game/Sounds/Fort_UI_Sounds/FrontEnd/UI_Cosmetic_Appear_Resin_01}}
 {id=WavePlayer_2, properties {soundWaveAssetPtr=/Game/Sounds/Fort_UI_Sounds/FrontEnd/UI_Cosmetic_Appear_Resin_02}}
 {id=WavePlayer_3, properties {soundWaveAssetPtr=/Game/Sounds/Fort_UI_Sounds/FrontEnd/UI_Cosmetic_Appear_Resin_03}}
 {id=WavePlayer_4, properties {soundWaveAssetPtr=/Game/Sounds/Fort_UI_Sounds/FrontEnd/UI_Cosmetic_Appear_Resin_04}}
 {id=WavePlayer_5, properties {soundWaveAssetPtr=/Game/Sounds/Fort_UI_Sounds/FrontEnd/UI_Cosmetic_Appear_Resin_05}}
 {id=WavePlayer_6, properties {soundWaveAssetPtr=/Game/Sounds/Fort_UI_Sounds/FrontEnd/UI_Cosmetic_Appear_Resin_06}}
]
properties {
 bOverrideConcurrency=true
 maxDistance=2097152
 soundClassObject=&Fort_UI
 virtualizationMode=Disabled
 volumeMultiplier=0.3
 attenuationOverrides {falloffDistance=1000, attenuationShapeExtents {x=0}}
 concurrencyOverrides {
  bVolumeScaleCanRelease=true
  maxCount=1
  resolutionRule=StopOldest
  retriggerTime=0.15
  voiceStealReleaseTime=0.05
  volumeScale=0.5
  volumeScaleAttackTime=0.1
  volumeScaleReleaseTime=0.25
 }
}
