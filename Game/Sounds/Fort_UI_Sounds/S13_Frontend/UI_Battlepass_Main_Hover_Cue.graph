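# UI sound cue for hovering on the main Battle Pass in the Season 13 front end, routed to the Fort_UI sound class.
# A mixer layers a shared button hover click at full volume with a frontend hover sound and a Battle Pass hover wave at half volume.
# Only the Battle Pass layer passes through a modulator for random variation.
# The whole cue is pitched up, and a shared hover-tick concurrency rule limits rapid overlapping hovers.
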SoundCue /Game/Sounds/Fort_UI_Sounds/S13_Frontend/UI_Battlepass_Main_Hover_Cue
root=Mixer_0
nodes [
 {id=Mixer_0, children [WavePlayer_2,WavePlayer_0,Modulator_1], properties {inputVolume [1,0.5,0.5]}}
 {id=Modulator_1, children [WavePlayer_5], properties {pitchMin=1}}
 {id=WavePlayer_0, properties {soundWaveAssetPtr=/Game/Sounds/Fort_UI_Sounds/FrontEnd/Fort_UI_CB_Hover}}
 {id=WavePlayer_2, properties {soundWaveAssetPtr=/Game/Sounds/Fort_UI_Sounds/FrontEnd/New_FrontEnd/UI_FN_Button_Hover_Click_02}}
 {id=WavePlayer_5, properties {soundWaveAssetPtr=&/S13_UI_Battlepass_Main_Hover}}
]
properties {
 maxDistance=2097152
 pitchMultiplier=1.25
 priority=0.9
 soundClassObject=&Fort_UI
 volumeMultiplier=1
 concurrencyOverrides {maxCount=1, resolutionRule=StopOldest}
 concurrencySet [/Game/Sounds/Concurrency_Rules/FrontEnd_UI_Menu_Max_01_Stop_Oldest_Hover_Tick]
}
